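; GA_Quest_S22_TranslatorSetup is a gameplay ability that credits the Season 22 translator-setup quest when activated.
; On activation it sends an "Interact" custom stat event tagged Athena.Quests.S22.TranslatorSetup.Placed to the avatar controller's quest manager, then ends immediately.
; On end it removes active effects with the Granted TranslatorSetupComplete tag from the avatar's ability system.
; Debug strings print on activation and on failed passive activation.
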(event Ability|EventActivateAbility
  (Development|PrintString "Stat event sent!" true true "(R=1.000000,G=0.000000,B=0.444432,A=1.000000)" 3.0)
  (Quest|SendCustomStatEventwithTags (Quests|GetQuestManager (Fortnite|GetFortPlayerControllerfromActor (Ability|GetAvatarActorfromActorInfo)) "Invalid") "Interact" "(GameplayTags=)" "(GameplayTags=((TagName=\"Athena.Quests.S22.TranslatorSetup.Placed\")))" 1)
  (Ability|EndAbility))

(event Ability|EventOnEndAbility
  (params (bWasCancelled :type Boolean))
  (|CallParentFunction bWasCancelled)
  (GameplayEffects|RemoveActiveEffectswithTags (Ability|GetFortAbilitySystemComponent (Ability|GetAvatarActorfromActorInfo)) "(GameplayTags=((TagName=\"Granted.Athena.Quest.S22.TranslatorSetupComplete\")))"))

(event Ability|EventFailedToActivatePassiveAbility
  (params (ActorInfo :type Gameplay%20Ability%20Actor%20Info%20Structure))
  (|CallParentFunction)
  (Development|PrintString "Failed to activate" true true "(R=1.000000,G=0.000000,B=0.444432,A=1.000000)" 3.0))
